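# Stereo one-shot UI MetaSound, named for front-end screen-wipe whooshes, routed to the Fort_UI sound class.
# It is a preset of MSS_UI_Button_Default_Parent, so the playback graph comes from that parent.
# It supplies a single shop tab-enter wave, a fixed pitch and a narrow min/max volume range as inputs.
# Its source properties lower overall volume and define single-voice, stop-oldest concurrency override values with a short retrigger time.
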MetaSoundSource /Game/Sounds/Fort_UI_Sounds/FrontEnd/Whooshes/MSS_UI_Whoosh_ScreenWipes
preset_parent=/Game/Sounds/Fort_UI_Sounds/FrontEnd/DefaultSounds/PressedButtons/MSS_UI_Button_Default_Parent
inputs [
 MaxPitch {type=Float}
 MaxVolume {type=Float}
 MinPitch {type=Float}
 MinVolume {type=Float}
 Sounds {type=WaveAsset:Array}
 UE.Source.OnPlay {type=Trigger}
]
interfaces [UE.OutputFormat.Stereo:1.0, UE.Source.OneShot:1.0, UE.Source:1.1]
outputs [
 "UE.OutputFormat.Stereo.Audio:0" {type=Audio}
 "UE.OutputFormat.Stereo.Audio:1" {type=Audio}
 UE.Source.OneShot.OnFinished {type=Trigger}
]
pages [Default {input_defaults {
   MaxPitch=1
   MaxVolume=1.2
   MinPitch=1
   MinVolume=1
   UE.Source.OnPlay=null
   Sounds [/Game/Sounds/Fort_UI_Sounds/FrontEnd/TopBarTabs/wav/UI_TobBar_Shop_Enter_01]
  }}]
properties {
 assetImportData=&self:AssetImportData
 numChannels=2
 outputFormat=Stereo
 sampleRate=48000
 soundClassObject=&Fort_UI
 virtualizationMode=Disabled
 volume=0.2
 concurrencyOverrides {maxCount=1, resolutionRule=StopOldest, retriggerTime=0.11}
 concurrencySet [null]
}
