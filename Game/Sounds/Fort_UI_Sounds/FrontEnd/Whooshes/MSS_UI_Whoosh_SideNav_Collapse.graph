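# One-shot stereo MetaSound for the front-end UI whoosh, named for collapsing the side nav.
# A random-sound subgraph plays the Sounds array, currently one screen-remove wave, with slight pitch and volume variation, starting slightly past the requested start time.
# A start-time comparison picks the attack time of an ADSR envelope that scales both channels, allowing a short fade-in.
# Concurrency keeps one instance, stopping the oldest, alongside a side-nav whoosh concurrency set.
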MetaSoundSource /Game/Sounds/Fort_UI_Sounds/FrontEnd/Whooshes/MSS_UI_Whoosh_SideNav_Collapse
inputs [
 MaxPitch {type=Float}
 MaxVolume {type=Float}
 MinPitch {type=Float}
 MinVolume {type=Float}
 Seed {type=Int32}
 Shared State Behavior {type=Enum:SharedStateBehaviorType}
 Sounds {type=WaveAsset:Array}
 UE.Source.OnPlay {type=Trigger}
 UE.Source.StartTime.StartTime {type=Time}
]
interfaces [UE.OutputFormat.Stereo:1.0, UE.Source.OneShot:1.0, UE.Source.StartTime:1.0, UE.Source:1.1]
outputs [
 "UE.OutputFormat.Stereo.Audio:0" {type=Audio}
 "UE.OutputFormat.Stereo.Audio:1" {type=Audio}
 UE.Source.OneShot.OnFinished {type=Trigger}
]
pages [Default {
  input_defaults {
   MaxPitch=1.05
   MaxVolume=1.2
   MinPitch=0.95
   MinVolume=1
   Seed=-1
   Shared State Behavior=1
   UE.Source.OnPlay=null
   UE.Source.StartTime.StartTime=0
   Sounds [/Game/Sounds/Fort_UI_Sounds/FrontEnd/Whooshes/wav/UI_Prompt_Screen_Remove_01]
  }
  nodes [
   n1 = &MSP_Branch(
    In=^UE.Source.OnPlay
    Value=/FortMetasound/Patches/CompareFloat(A=UE.ConversionTimeToFloat(In=^UE.Source.StartTime.StartTime).Out) {also {defaults {Type=3}}}.Result
   )
   TriggerRouteTriggerRouteTime2_0 = "TriggerRoute.Trigger Route (Time, 2):Time"(Set 1=@n1.False, Set 0=@n1.True) {also {defaults {Value 0=0.1}}}
   ADSREnvelopeADSREnvelope_0 = ADSR Envelope.ADSR Envelope:Float(Trigger Attack=@TriggerRouteTriggerRouteTime2_0.On Set, Attack Time=@TriggerRouteTriggerRouteTime2_0.Value) {also {defaults {Decay Time=0, Sustain Level=1}}}
   MSS_PlayRandomSound_Stereo_0 = &MSS_PlayRandomSound_Stereo(
    MaxPitch=^MaxPitch
    MaxVolume=^MaxVolume
    MinPitch=^MinPitch
    MinVolume=^MinVolume
    Sounds=^Sounds
    UE.Source.OnPlay=^UE.Source.OnPlay
    Start Time=UE.Add:Time(PrimaryOperand=^UE.Source.StartTime.StartTime) {also {defaults {AdditionalOperands=0.07}}}.Out
   )
   Multiply_0 = UE.Multiply:Audio by Float(AdditionalOperands=@ADSREnvelopeADSREnvelope_0.Out Envelope, PrimaryOperand=@MSS_PlayRandomSound_Stereo_0."UE.OutputFormat.Stereo.Audio:0")
   Multiply_1 = UE.Multiply:Audio by Float(AdditionalOperands=@ADSREnvelopeADSREnvelope_0.Out Envelope, PrimaryOperand=@MSS_PlayRandomSound_Stereo_0."UE.OutputFormat.Stereo.Audio:1")
  ]
  links [
   {dst="@UE.Source.OneShot.OnFinished", src=MSS_PlayRandomSound_Stereo_0.UE.Source.OneShot.OnFinished}
   {dst="@UE.OutputFormat.Stereo.Audio:0", src=Multiply_0.Out}
   {dst="@UE.OutputFormat.Stereo.Audio:1", src=Multiply_1.Out}
  ]
 }]
properties {
 assetImportData=&self:AssetImportData
 numChannels=2
 outputFormat=Stereo
 sampleRate=48000
 soundClassObject=&Fort_UI
 virtualizationMode=Disabled
 volume=0.115
 concurrencyOverrides {maxCount=1, resolutionRule=StopOldest, retriggerTime=0.11}
 concurrencySet [&/CONC_Whoosh_SideNav]
}
